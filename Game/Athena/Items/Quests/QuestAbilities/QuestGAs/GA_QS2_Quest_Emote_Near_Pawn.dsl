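; Quest gameplay ability that reports player pawns near the ability's owning player, evidently when they emote.
; On event activation it calls the parent, then finds Fort player pawns within a configurable meter radius of the owner's pawn.
; Each pawn found gets its own CustomVerb.EmoteNearPawn message, tagged via the "PawnNear" subject, broadcast for quest filters to match.
; Despite the "NearbyEnemies" function name, no team filtering is applied here.
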(event Ability|EventActivateAbilityFromEvent
  (params (EventData :type Gameplay%20Event%20Data%20Structure%20%28by%20ref%29))
  ;[1 1 1] When using this in Gameplay Abilities and Effects Component you will use a custom filter, with the CusotmVerb.EmoteNearPawn and set the subject to be "PawnNear"
  (|CallParentFunction EventData)
  (CallFunction|CheckforNearbyEnemies))

(event Custom|CheckForNearbyEnemies
  ;[1 1 1] When using this in Gameplay Abilities and Effects Component you will use a custom filter, with the CusotmVerb.EmoteNearPawn and set the subject to be "PawnNear"
  ;;[1 1 1] Check for nearby Fort Player Pawns
  (bind _asfort_player_state (Utilities|Casting|CastToFortPlayerState (Ability|GetOwningActorfromActorInfo)))
  (bind _outfortplayerpawns (Fortnite|GetFortPlayerPawnsInRange (Transformation|GetActorLocation (PlayerState|GetPawn _asfort_player_state)) (* (Variables|Default|GetRadiusInMeters) 1e+02)))
  ;[1 1 1] Send each found pawn
  (bind _array_element (Utilities|Array|ForEachLoopwithBreak _outfortplayerpawns)
    (:LoopBody
      (CallFunction|SendFoundPawn _array_element))))

(event Custom|SendFoundPawn
  (params (FoundPawn :type Actor%20Object%20Reference))
  ;[1 1 1] When using this in Gameplay Abilities and Effects Component you will use a custom filter, with the CusotmVerb.EmoteNearPawn and set the subject to be "PawnNear"
  (bind _result (CustomVerb|AddTagsfromSubject (CustomVerb|MakeCustomVerbMessage (Ability|GetOwningActorfromActorInfo) "(TagName=\"CustomVerb.EmoteNearPawn\")") "PawnNear" FoundPawn))
  (CustomVerb|BroadcastCustomVerbMessage _result))
