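; Quest gameplay ability for Battle Royale that tracks damaging enemy players with different weapons in quick succession.
; It switches on the triggering event tag: weapon switches store the active weapon, and damage events ignore non-player-pawn or same-team targets.
; If the owner has the timer tag and the weapon differs from the last recorded one, it sends a quest stat event.
; Otherwise it records the weapon and applies the quest's GameplayEffect to the owner.
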(event Ability|EventActivateAbilityFromEvent
  (params (EventData :type Gameplay%20Event%20Data%20Structure%20%28by%20ref%29))
  (bind (_eventtag _instigator _target _optionalobject) (Utilities|Struct|BreakGameplayEventData EventData))
  (bind (_eventtag_2 _instigator_2 _target_2) (Utilities|Struct|BreakGameplayEventData EventData))
  (bind _owningplayerpawn (Variables|GAVars|Data|GetOwningPlayerPawn))
  (|CallParentFunction EventData)
  (switch Utilities|FlowControl|Switch|SwitchonGameplayTag _eventtag_2
    (:Default
      (Ability|EndAbility))
    (:Event.Weapons.WeaponSwitch
      ;[1 1 1] Weapon switch
      (bind _asactor (Utilities|Casting|CastToActor _optionalobject)
        (:then
          (Variables|Default|SetActiveWeaponActor _asactor))
        (:CastFailed
          _))
      (Ability|EndAbility))
    (:Event.Damage.Dealt
      ;[1 1 1] Damage dealt
      (bind _asfort_player_pawn_athena (Utilities|Casting|CastToFortPlayerPawnAthena _target_2)
        (:then
          (if (Team|OnSameTeam _asfort_player_pawn_athena _owningplayerpawn)
            (Ability|EndAbility)
            (elif (and (Ability|DoesTargetHaveTags _owningplayerpawn (Variables|Default|GetTimerTag)) (Utilities|NotEqual(Object) (Variables|Default|GetActiveWeaponActor) (Variables|Default|GetLastActiveWeaponActor)))
              (Quest|SendComplexCustomStatEvent (Variables|GAVars|Data|GetOwningPlayerQuestManager) 0 "(GameplayTags=)" "(GameplayTags=((TagName=\"Athena.Quests.DamageWithDifferentWeaponsWithinTimer\")))" 1)
              (Ability|EndAbility)
              (else
                (Variables|Default|SetLastActiveWeaponActor (Variables|Default|GetActiveWeaponActor))
                (Ability|ApplyGameplayEffectToOwner "/Game/Athena/Items/Quests/QuestAbilities/QuestGEs/GE_Quest_BR_Damage_Different_Weapons_Quickly.GE_Quest_BR_Damage_Different_Weapons_Quickly_C")
                (Ability|EndAbility)))))
        (:CastFailed
          (Ability|EndAbility))))))
