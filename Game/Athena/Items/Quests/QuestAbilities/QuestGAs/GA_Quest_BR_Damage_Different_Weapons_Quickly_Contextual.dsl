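; Battle Royale quest ability for the contextual "damage with different weapons quickly" quest, flagging it as relevant when the player carries enough weapons.
; On event activation, it calls the parent, then checks the owning player's controller inventory for two Weapon items.
; If the check passes, it appends temporary relevancy tags through the Athena quest manager component, if valid.
; A failed check or controller cast ends the ability; the success path does not end it here.
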(event Ability|EventActivateAbilityFromEvent
  (params (EventData :type Gameplay%20Event%20Data%20Structure%20%28by%20ref%29))
  (bind _owningplayerquestmanager (Variables|GAVars|Data|GetOwningPlayerQuestManager))
  (bind _returnvalue (Class|FortQuestManager|GetQuestManagerComponent _owningplayerquestmanager "/Script/FortniteGame.FortQuestManagerComponent_Athena"))
  (|CallParentFunction EventData)
  (bind _asfort_player_controller_gameplay (Utilities|Casting|CastToFortPlayerControllerGameplay (Pawn|GetController (Variables|GAVars|Data|GetOwningPlayerPawn)))
    (:then
      (bind _returnvalue_2 (Inventory|CheckPossessionOfAmountOfItems _asfort_player_controller_gameplay "Weapon" 2))
      (if _returnvalue_2
        (Utilities|IsValid _returnvalue
          (:Is%20Valid
            (Quest|AppendTemporaryRelevancyTags :self _returnvalue))
          (:Is%20Not%20Valid
            _))
        (Ability|EndAbility)))
    (:CastFailed
      (Ability|EndAbility))))
