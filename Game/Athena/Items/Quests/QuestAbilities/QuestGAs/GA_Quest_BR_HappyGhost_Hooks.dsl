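; Quest gameplay ability that reports quest stat events when the owning player hooks something with the Happy Ghost harpoon.
; It casts the event target to send PulledPlayer for non-teammates, PulledWeapon for pickups, or a harpoon fishing catch for flopper spawners.
; For spawners, it waits slightly longer than the spawn delay read from the payload's projectile, then sends PulledFish if the spawned item has the fish tag.
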(event Ability|EventActivateAbilityFromEvent
  (params (EventData :type Gameplay%20Event%20Data%20Structure%20%28by%20ref%29))
  (bind (_eventtag _instigator _target _optionalobject) (Utilities|Struct|BreakGameplayEventData EventData))
  (|CallParentFunction EventData)
  (Utilities|FlowControl|Sequence
    (:then_0
      ;[1 1 1] Need happy ghost projectile to send itself in the payload so we can dynamically get the spawn delay time
      (bind _asprj_athena_happy_ghost (Utilities|Casting|CastToPrj_Athena_HappyGhost _optionalobject))
      (Variables|Default|SetItemSpawnDelay (+ 0.01 (Variables|/Game/Athena/Items/Consumables/HappyGhost/Prj_Athena_HappyGhost.Prj_Athena_HappyGhost_C|GetItemSpawnDelay _asprj_athena_happy_ghost))))
    (:then_1
      (bind _asfort_player_pawn (Utilities|Casting|CastToFortPlayerPawn _target)
        (:then
          (if (not (Team|OnSameTeam (Variables|GAVars|Data|GetOwningPlayerPawn) _asfort_player_pawn))
            (Quest|SendComplexCustomStatEvent (Variables|GAVars|Data|GetOwningPlayerQuestManager) 0 "(GameplayTags=)" "(GameplayTags=((TagName=\"Athena.Quests.HappyGhost.PulledPlayer\")))" 1)))
        (:CastFailed
          (Utilities|Casting|CastToFortPickup _target
            (:then
              (Quest|SendComplexCustomStatEvent (Variables|GAVars|Data|GetOwningPlayerQuestManager) 0 "(GameplayTags=)" "(GameplayTags=((TagName=\"Athena.Quests.HappyGhost.PulledWeapon\")))" 1))
            (:CastFailed
              (bind _asbga_athena_flopper_spawn_parent (Utilities|Casting|CastToBGA_Athena_FlopperSpawn_Parent _target))
              (bind _output_get (Variables|Default|SetFishSpawner _asbga_athena_flopper_spawn_parent))
              (Quest|SendComplexCustomStatEvent (Variables|GAVars|Data|GetOwningPlayerQuestManager) 0 "(GameplayTags=)" "(GameplayTags=((TagName=\"Athena.Quests.FishingCatchItem.Harpoon\")))" 1)
              ;[1 1 1] The fish spawner has a delay before it spawns the item, so we need to wait before we can access it
              (Utilities|FlowControl|Delay (Variables|Default|GetItemSpawnDelay))
              (if (Utilities|IsValid (Variables|/Game/Athena/Items/EnvironmentalItems/FlopperSpawn/BGA_Athena_FlopperSpawn_Parent.BGA_Athena_FlopperSpawn_Parent_C|GetSpawnedItem _output_get) :!class /Script/Engine.KismetSystemLibrary)
                (if (GameplayTags|HasMatchingGameplayTag (Class|FortItemFunctionLibrary|BreakFortItemEntry (Pickup|GetItemEntry (Variables|/Game/Athena/Items/EnvironmentalItems/FlopperSpawn/BGA_Athena_FlopperSpawn_Parent.BGA_Athena_FlopperSpawn_Parent_C|GetSpawnedItem _output_get) 0)) (Variables|ReadOnly|GetFishItemTag))
                  (Quest|SendComplexCustomStatEvent (Variables|GAVars|Data|GetOwningPlayerQuestManager) 0 "(GameplayTags=)" "(GameplayTags=((TagName=\"Athena.Quests.HappyGhost.PulledFish\")))" 1))))))))))
