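; GA_Quest_BR_RecentlyLandedFromBus is a quest gameplay ability, likely marking a player as recently landed from the battle bus in Battle Royale.
; It activates from a gameplay event, calls the parent implementation, applies the gameplay effect held in its GE variable to the owner, then ends immediately.
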(event Ability|EventActivateAbilityFromEvent
  (params (EventData :type Gameplay%20Event%20Data%20Structure%20%28by%20ref%29))
  (|CallParentFunction EventData)
  (Ability|ApplyGameplayEffectToOwner (Variables|Default|GetGE))
  (Ability|EndAbility))
