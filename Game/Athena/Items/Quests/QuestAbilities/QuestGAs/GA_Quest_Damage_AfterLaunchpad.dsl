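; GA_Quest_Damage_AfterLaunchpad is a quest gameplay ability, apparently supporting a quest about dealing damage after using a launchpad.
; When activated from a gameplay event, it calls the parent implementation and applies the GE_Quest_Damage_AfterLaunchpad_Tag effect to its owner.
; It then ends immediately, leaving that effect as its only added behavior.
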(event Ability|EventActivateAbilityFromEvent
  (params (EventData :type Gameplay%20Event%20Data%20Structure%20%28by%20ref%29))
  (|CallParentFunction EventData)
  (Ability|ApplyGameplayEffectToOwner "/Game/Athena/Items/Quests/QuestAbilities/QuestGEs/GE_Quest_Damage_AfterLaunchpad_Tag.GE_Quest_Damage_AfterLaunchpad_Tag_C")
  (Ability|EndAbility))

(event Ability|EventOnEndAbility
  (params (bWasCancelled :type Boolean)))
