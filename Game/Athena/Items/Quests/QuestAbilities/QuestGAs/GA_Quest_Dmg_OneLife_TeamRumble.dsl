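; Quest gameplay ability, named for a damage-in-one-life Team Rumble quest, with most of its logic empty.
; InitQuestManagerCheck resolves the instigator's player controller and checks once for the configured quest and objective, caching that it checked.
; It ends the ability if the controller is invalid or the quest is missing, also clearing the active flag in the latter case.
; Event activation and CheckQuestCredit have no body, and ShouldAbilityRespondToEvent always returns false.
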(event Ability|EventActivateAbilityFromEvent
  (params (EventData :type Gameplay%20Event%20Data%20Structure%20%28by%20ref%29)))

(fn CheckQuestCredit
  (params (Magnitude_Damage :type Float%20%28double-precision%29)))

(fn InitQuestManagerCheck
  (bind _self self)
  (bind _returnvalue (Ability|GetContextfromOwner :self _self))
  (bind _returnvalue_2 (Ability|EffectContext|GetInstigatorActor _returnvalue))
  (bind _returnvalue_3 (Fortnite|GetFortPlayerControllerfromActor _returnvalue_2))
  (if (Variables|Default|GetbHasCheckedQuestManager)
    (return)
    (elif (Utilities|IsValid _returnvalue_3 :!class /Script/Engine.KismetSystemLibrary)
      (bind _questvalid (Quest|DoesPlayerHaveQuest _returnvalue_3 (Variables|Default|GetQuestDefinition) (Variables|Default|GetObjective_BackendName)))
      (if _questvalid
        (Variables|Default|SetbHasCheckedQuestManager true)
        (return)
        (else
          (Variables|Default|SetbHasCheckedQuestManager true)
          (Variables|Default|SetbQuestActive false)
          (Ability|EndAbility)))
      (else
        (Ability|EndAbility)))))

(fn ShouldAbilityRespondToEvent
  (params (ActorInfo :type Gameplay%20Ability%20Actor%20Info%20Structure) (Payload :type Gameplay%20Event%20Data%20Structure))
  (return false))
