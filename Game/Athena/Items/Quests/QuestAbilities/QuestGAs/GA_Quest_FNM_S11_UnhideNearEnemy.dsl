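; Quest gameplay ability for Season 11 FNM that reports the Athena.Quests.UnhideWithEnemyClose stat when triggered by an event with enemy players nearby.
; After calling its parent, it sphere-overlaps player pawns within EnemyCheckRadius of the owning pawn, excluding the owner.
; Each pawn on another team prints a debug "BOO!!!!" and sends the stat to the owner's quest manager; the loop never breaks, so several enemies count separately.
; The ability then ends.
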(event Ability|EventActivateAbilityFromEvent
  (params (EventData :type Gameplay%20Event%20Data%20Structure%20%28by%20ref%29))
  (bind _owningplayerpawn (Variables|GAVars|Data|GetOwningPlayerPawn))
  (|CallParentFunction EventData)
  (bind _outactors (Collision|SphereOverlapActors (Transformation|GetActorLocation (Variables|GAVars|Data|GetOwningPlayerPawn)) (Variables|Default|GetEnemyCheckRadius) (Utilities|Array|MakeArray "ObjectTypeQuery3") "/Script/FortniteGame.FortPlayerPawnAthena" (Utilities|Array|MakeArray (Variables|GAVars|Data|GetOwningPlayerPawn))))
  (bind _array_element (Utilities|Array|ForEachLoopwithBreak _outactors)
    (:LoopBody
      (if (not (Team|OnSameTeam _array_element _owningplayerpawn))
        (Development|PrintString "BOO!!!!" true true "(R=0.280817,G=1.000000,B=0.000000,A=1.000000)")
        (Quest|SendComplexCustomStatEvent (Quests|GetQuestManager (Fortnite|GetFortPlayerControllerfromActor _owningplayerpawn) "Invalid") 0 "(GameplayTags=((TagName=\"Athena.Quests.UnhideWithEnemyClose\")))" "(GameplayTags=)" 1)))
    (:Completed
      _))
  (Ability|EndAbility))
