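; Quest gameplay ability for the Mystery quest's rift zone, applying a rift zone gameplay effect to its owner under a tag condition.
; On event activation it calls the parent, then checks the owning player pawn for the Zero Point or Pink Oatmeal tags.
; If either check passes, it applies the effect and prints a debug string.
; The ability then ends.
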(event Ability|EventActivateAbilityFromEvent
  (params (EventData :type Gameplay%20Event%20Data%20Structure%20%28by%20ref%29))
  (|CallParentFunction EventData)
  (bind _bdidactorhavetags (Class|AthenaQuestFunctionLibrary|AthenaCheckActorHasTags (Variables|GAVars|Data|GetOwningPlayerPawn) (Variables|Default|GetZeroPointTag) true))
  (bind _bdidactorhavetags_2 (Class|AthenaQuestFunctionLibrary|AthenaCheckActorHasTags (Variables|GAVars|Data|GetOwningPlayerPawn) (Variables|Default|GetPinkOatmealTag)))
  (if (or _bdidactorhavetags_2 _bdidactorhavetags)
    (Ability|ApplyGameplayEffectToOwner (Variables|Default|GetRiftZoneGE))
    (Development|PrintString "RIFT: GE applied"))
  (Ability|EndAbility))
