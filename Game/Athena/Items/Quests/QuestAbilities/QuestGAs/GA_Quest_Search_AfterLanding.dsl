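; GA_Quest_Search_AfterLanding is an event-activated quest gameplay ability that marks its owner for search-after-landing quest tracking, judging by its name and effects.
; On activation it calls the parent handler, then applies the AfterLanding30 and Search_AfterLanding tag gameplay effects to the owner and immediately ends.
; The trigger, effect durations and granted tags are defined outside this graph.
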(event Ability|EventActivateAbilityFromEvent
  (params (EventData :type Gameplay%20Event%20Data%20Structure%20%28by%20ref%29))
  (|CallParentFunction EventData)
  (Ability|ApplyGameplayEffectToOwner "/Game/Athena/Items/Quests/QuestAbilities/QuestGEs/GE_Quest_AfterLanding30_Tag.GE_Quest_AfterLanding30_Tag_C")
  (Ability|ApplyGameplayEffectToOwner "/Game/Athena/Items/Quests/QuestAbilities/QuestGEs/GE_Quest_Search_AfterLanding_Tag.GE_Quest_Search_AfterLanding_Tag_C")
  (Ability|EndAbility))
