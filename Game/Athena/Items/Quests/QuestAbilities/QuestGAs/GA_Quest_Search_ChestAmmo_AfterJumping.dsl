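; Quest gameplay ability that tracks searching chest and ammo box pairs after jumping, judged by an after-landing tag on the instigator.
; It responds only when that tag is present and the target is tagged as a chest or ammo box.
; Activation increments the matching counter, then compares both counts to detect a newly completed pair.
; Each new pair sends a stat event to the owning player's quest manager.
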(event Ability|EventActivateAbilityFromEvent
  (params (EventData :type Gameplay%20Event%20Data%20Structure%20%28by%20ref%29))
  (bind (_eventtag _instigator _target _optionalobject _optionalobject2 _contexthandle _instigatortags _targettags) (Utilities|Struct|BreakGameplayEventData EventData))
  (bind _t_chest (Variables|GAVars|Setup|GetT_Chest))
  (bind _case_0_true (GameplayTags|TagContainer|CompareTagContainertoOtherTags _t_chest _targettags))
  (|CallParentFunction EventData)
  ;[1 1 1] Increment appropriate counter
  (Math|Integer|IncrementInt (select _case_0_true (Variables|GAVars|Data|GetChestCount) (Variables|GAVars|Data|GetAmmoBoxCount))
    (:%20%20
      ;[1 1 1] Check if player has completed a new set (chest + ammo box)
      (if (select _case_0_true (>= (Variables|GAVars|Data|GetAmmoBoxCount) (Variables|GAVars|Data|GetChestCount)) (>= (Variables|GAVars|Data|GetChestCount) (Variables|GAVars|Data|GetAmmoBoxCount)))
        ;[1 1 1] Send quest update
        (Quest|SendComplexCustomStatEvent (Variables|GAVars|Data|GetOwningPlayerQuestManager) 0 "(GameplayTags=((TagName=\"Athena.Quests.AbilityTrigger.SearchChestAmmoAfterJumping\")))" "(GameplayTags=)" 1)
        (else
          _))
      (Ability|EndAbility))))

(fn ShouldAbilityRespondToEvent
  (params (ActorInfo :type Gameplay%20Ability%20Actor%20Info%20Structure) (Payload_EventTag :type Gameplay%20Tag%20Structure) (Payload_Instigator :type Actor%20Object%20Reference) (Payload_Target :type Actor%20Object%20Reference) (Payload_OptionalObject :type Object%20Reference) (Payload_OptionalObject2 :type Object%20Reference) (Payload_ContextHandle :type Gameplay%20Effect%20Context%20Handle%20Structure) (Payload_InstigatorTags :type Gameplay%20Tag%20Container%20Structure) (Payload_TargetTags :type Gameplay%20Tag%20Container%20Structure) (Payload_EventMagnitude :type Float%20%28single-precision%29) (Payload_TargetData :type Gameplay%20Ability%20Target%20Data%20Handle%20Structure) (Payload_InstancedEventData :type Instanced%20Struct%20Structure))
  (bind (_case_0_true _case_1_true) (GameplayTags|TagContainer|CompareTagContainertoOtherTags (Variables|GAVars|Setup|GetT_Chest) (Variables|GAVars|Setup|GetT_AmmoBox) Payload_TargetTags))
  (return (and (GameplayTags|TagContainer|CompareTagContainertoOtherTags (Variables|GAVars|Setup|GetT_AfterLanding) Payload_InstigatorTags) (or _case_0_true _case_1_true))))
