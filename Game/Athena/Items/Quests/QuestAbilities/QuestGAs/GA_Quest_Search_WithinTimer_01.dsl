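; Quest gameplay ability that credits a player for searching a set number of treasure chests within a time limit.
; It responds only to events whose target carries the treasure chest tag, then trims a full timestamp array and appends the current server world time.
; When the array is populated and its first-to-last span fits TimeToComplete, it sends the quest stat event to the owning player's quest manager, then ends.
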(event Ability|EventActivateAbilityFromEvent
  (params (EventData :type Gameplay%20Event%20Data%20Structure%20%28by%20ref%29))
  (bind _timestamps (Variables|GAVars|Data|GetTimeStamps))
  (|CallParentFunction EventData)
  (if (CallFunction|IsTimeArrayPopulated)
    (for _index (range (+ (- (Utilities|Array|Length (Variables|GAVars|Data|GetTimeStamps)) (Variables|GAVars|Setup|GetNumberToSearch)) 1))
      (Utilities|Array|RemoveIndex _timestamps _index))
    (else
      _))
  (Utilities|Array|Add _timestamps (Fortnite|Sync|Time|GetServerWorldTimeSeconds))
  (if (CallFunction|AreTimesWithinLimit)
    (Quest|SendComplexCustomStatEvent (Variables|GAVars|Data|GetOwningPlayerQuestManager) 0 (Variables|GAVars|Setup|GetTC_Quest) "(GameplayTags=)" 1)
    (else
      _))
  (Ability|EndAbility))

(fn AreTimesWithinLimit
  (return (and (CallFunction|IsTimeArrayPopulated) (<= (- (Utilities|Array|Get(acopy) (Variables|GAVars|Data|GetTimeStamps) (Utilities|Array|LastIndex (Variables|GAVars|Data|GetTimeStamps))) (Utilities|Array|Get(acopy) (Variables|GAVars|Data|GetTimeStamps))) (Variables|GAVars|Setup|GetTimeToComplete)))))

(fn IsTimeArrayPopulated
  (return (>= (Utilities|Array|Length (Variables|GAVars|Data|GetTimeStamps)) (Variables|GAVars|Setup|GetNumberToSearch))))

(fn ShouldAbilityRespondToEvent
  (params (ActorInfo :type Gameplay%20Ability%20Actor%20Info%20Structure) (Payload_EventTag :type Gameplay%20Tag%20Structure) (Payload_Instigator :type Actor%20Object%20Reference) (Payload_Target :type Actor%20Object%20Reference) (Payload_OptionalObject :type Object%20Reference) (Payload_OptionalObject2 :type Object%20Reference) (Payload_ContextHandle :type Gameplay%20Effect%20Context%20Handle%20Structure) (Payload_InstigatorTags :type Gameplay%20Tag%20Container%20Structure) (Payload_TargetTags :type Gameplay%20Tag%20Container%20Structure) (Payload_EventMagnitude :type Float%20%28single-precision%29) (Payload_TargetData :type Gameplay%20Ability%20Target%20Data%20Handle%20Structure) (Payload_InstancedEventData :type Instanced%20Struct%20Structure))
  (bind _returnvalue (GameplayTags|HasMatchingGameplayTag(Message) Payload_Target "(TagName=\"Building.Type.Container.TreasureChest\")"))
  (return _returnvalue))
