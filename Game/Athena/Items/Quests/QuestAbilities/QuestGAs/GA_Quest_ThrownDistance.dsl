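; Quest gameplay ability that measures how far a thrown projectile travels and reports it for quest tracking.
; On event activation it caches instigator and target tags, the target's starting location, and binds to its Fort projectile movement component's stop event.
; When the projectile stops, it unbinds, broadcasts a custom verb message with those tags and the distance scaled by 0.01 (centimeters to meters), then ends.
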(event Ability|EventActivateAbilityFromEvent
  (params (EventData_EventTag :type Gameplay%20Tag%20Structure) (EventData_Instigator :type Actor%20Object%20Reference) (EventData_Target :type Actor%20Object%20Reference) (EventData_OptionalObject :type Object%20Reference) (EventData_OptionalObject2 :type Object%20Reference) (EventData_ContextHandle :type Gameplay%20Effect%20Context%20Handle%20Structure) (EventData_InstigatorTags :type Gameplay%20Tag%20Container%20Structure) (EventData_TargetTags :type Gameplay%20Tag%20Container%20Structure) (EventData_EventMagnitude :type Float%20%28single-precision%29) (EventData_TargetData :type Gameplay%20Ability%20Target%20Data%20Handle%20Structure) (EventData_InstancedEventData :type Instanced%20Struct%20Structure))
  ;[1 1 1] Send custom Verb Message when the object has stopped moving.
  (bind _output_get (Variables|Default|SetTargetTags EventData_InstigatorTags))
  (GameplayTags|AppendGameplayTagContainers _output_get EventData_TargetTags)
  (Variables|Default|SetStartingLocation (Transformation|GetActorLocation EventData_Target))
  (bind _asfort_projectile_base (Utilities|Casting|CastToFortProjectileBase EventData_Target))
  (bind _output_get_2 (Variables|Default|SetProjMovementcomp (Projectile|GetFortProjectileMovementComponent _asfort_projectile_base)))
  (EventDispatchers|BindEventtoOnProjectileStop _output_get_2 (delegate Custom|BroadcastVerbMessageOnStop)))

(event Custom|BroadcastVerbMessageOnStop
  (params (ImpactResult :type Hit%20Result%20Structure%20%28by%20ref%29))
  (bind _projmovementcomp (Variables|Default|GetProjMovementcomp))
  ;[1 1 1] Send custom Verb Message when the object has stopped moving.
  (EventDispatchers|UnbindEventfromOnProjectileStop _projmovementcomp (delegate Custom|BroadcastVerbMessageOnStop))
  (bind _result (CustomVerb|AddTagsforSubject (CustomVerb|MakeCustomVerbMessage (Ability|GetAvatarActorfromActorInfo) (Variables|Default|GetQuestTag)) "ItemThrown" (Variables|Default|GetTargetTags)))
  (bind _result_2 (CustomVerb|AddValue _result "Distance" (* (Math|Vector|Distance(Vector) (Variables|Default|GetStartingLocation) (Transformation|GetActorLocation (Components|GetOwner _projmovementcomp))) 0.01)))
  (CustomVerb|BroadcastCustomVerbMessage _result_2)
  (Ability|EndAbility))

(event Ability|EventOnEndAbility
  (params (bWasCancelled :type Boolean)))
